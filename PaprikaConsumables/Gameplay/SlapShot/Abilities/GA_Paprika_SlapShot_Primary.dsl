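; Primary-use gameplay ability for the Paprika SlapShot consumable that gives nearby targets shields and low gravity.
; After the montage trigger commits, it runs area targeting and gives the user low gravity.
; Targets receive the shield spec cached from the parent's SetupGE plus the low-gravity effect.
; The user gets a berry explosion cue; user and targets get an applied cue, plus a shields cue if current shield is below the Shield attribute.
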(event Custom|ApplyEffectInArea
  (bind _playerpawn (Variables|Default|GetPlayerPawn))
  ;[0.03429 0.3069 0.7899] Apply gameplay effect to nearby targets
  (bind (_asynctaskproxy _targetdata) (Ability|Tasks|DoTargetSelection "None" "(TagName=\"Targeting.ApplicationTag.0\")" true)
    (:then
      (GameplayCue|ExecuteGameplayCueOnActor(Burst) _playerpawn (Variables|Default|GetGC_BerryExplosion) (Ability|GameplayCue|MakeGameplayCueParameters 0.0))
      (CallFunction|ApplyCosmeticEffectstoActor _playerpawn))
    (:Targeted
      (Ability|ApplyGameplayEffectSpecToTarget (Variables|Default|GetGE_ShieldsGrantedSpec) _targetdata)
      (Ability|ApplyGameplayEffectToTarget _targetdata (Variables|Default|GetGE_LowGravityClass))
      (for _array_element (Ability|TargetData|GetAllActorsfromTargetData _targetdata)
        (CallFunction|ApplyCosmeticEffectstoActor _array_element)))))

(fn ApplyCosmeticEffectsToActor
  (params (Actor :type Actor%20Object%20Reference))
  (bind _returnvalue (Ability|GameplayCue|MakeGameplayCueParameters 0.0))
  (GameplayCue|ExecuteGameplayCueOnActor(Burst) Actor (Variables|Default|GetGC_EffectApplied) _returnvalue)
  (if (CallFunction|CheckMissingShields Actor)
    (GameplayCue|ExecuteGameplayCueOnActor(Burst) Actor (Variables|Default|GetGC_ShieldsApplied) _returnvalue)))

(fn CheckMissingShields
  (params (Actor :type Actor%20Object%20Reference))
  (bind (_bsuccessfullyfoundattribute _returnvalue) (Ability|Attribute|GetFloatAttribute Actor "(AttributeName=\"Shield\",Attribute=/Script/FortniteGame.FortHealthSet:Shield,AttributeOwner=\"/Script/CoreUObject.Class'/Script/FortniteGame.FortHealthSet'\")"))
  (bind (_bsuccessfullyfoundattribute_2 _returnvalue_2) (Ability|Attribute|GetFloatAttribute Actor "(AttributeName=\"CurrentShield\",Attribute=/Script/FortniteGame.FortHealthSet:CurrentShield,AttributeOwner=\"/Script/CoreUObject.Class'/Script/FortniteGame.FortHealthSet'\")"))
  (return (< _returnvalue_2 _returnvalue)))

(fn OnMontageTriggeredPostCommitSuccess
  (|CallParentFunction)
  (CallFunction|ApplyEffectinArea)
  (Ability|ApplyGameplayEffectToOwner (Variables|Default|GetGE_LowGravityClass))
  (return))

(fn SetupGE
  ;[0.4895 0.1317 1] Save the healing GE as a variable
  (bind _effect_out (|CallParentFunction))
  (Variables|Default|SetGE_ShieldsGrantedSpec _effect_out)
  (return _effect_out))
